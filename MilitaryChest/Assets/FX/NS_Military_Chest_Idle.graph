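# Niagara idle effect for the military chest, in practice a steady stream of small, short-lived sprite sparkles tinted by User.Color.
# Sparkles spawn along a box line shifted by User.GlobalOffset, then drift under gravity, curl noise and drag while curves fade and shrink them.
# A disabled Rays mesh emitter uses User.Brightness, User.RaysLength, time-of-day scaling and distance-based alpha fade.
# The Seam emitter has no renderers.
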NiagaraSystem /MilitaryChest/Assets/FX/NS_Military_Chest_Idle
dependencies {
 usedDataInterfaces [&NiagaraDataInterfaceCurve, &NiagaraDataInterfaceColorCurve]
 usedDynamicInputs []
 usedModules [&SystemState]
}
emitters {
 Rays {
  bEnabled=false
  emitterUpdateScript [
   &EmitterState(Life Cycle Mode=::NewEnumerator1, Loop Behavior=::NewEnumerator1, Loop Duration Mode=::NewEnumerator1, UseLoopDelay=false)
   &SpawnBurst_Instantaneous(Use Spawn Probability=false, UseLoopCountLimit=false)
   /SetVariables(Emitter.PanSpeed=0.25) {isSetParameters=true}
   /SetVariables(
    Emitter.EndColor:LinearColor=&MultiplyLinearColor(
     Linear Color=<rgba 0.3643 0 0.974 1>
     Scale Factor=&MakeFloat4FromVectorAndFloat("Vector (RGB)"=&VectorFromFloat(Value=2), "Float (Alpha)"=1)
    )
    Emitter.StartColor:LinearColor=&MultiplyLinearColor(
     Linear Color=<rgba 0.7258 0 0.974 1>
     Scale Factor=&MakeFloat4FromVectorAndFloat("Vector (RGB)"=&VectorFromFloat(Value=20), "Float (Alpha)"=1)
    )
   ) {isSetParameters=true}
  ]
  particleSpawnScript [
   &InitializeParticle(
    Color=$User.Color
    Mesh Scale=<0.55 1.7 0.7>
    Mesh Scale Mode=::NewEnumerator4
    Position Offset=<0 0 50>
    UsePositionOffset=true
    Write Material Random=true
   )
   &InitialMeshOrientation(Mesh Orientation Mode=::NewEnumerator2, Rotation=<0 0 0.25>, Use Rotation Vector=true)
   &InitialMeshOrientation(Mesh Orientation Mode=::NewEnumerator2, Rotation=<0.5 0 0>, Use Rotation Vector=true) {enabled=false, module=InitialMeshOrientation001}
  ]
  particleUpdateScript [
   &ParticleState(Kill Particles When Lifetime Has Elapsed=false, Let Infinitely Lived Particles Die When Emitter Deactivates=true)
   &DynamicMaterialParameters(
    Index 0 Param 1=0.8
    Index 0 Param 2=1.2
    Index 0 Param 3=$User.RaysLength
    Index 0 Param 4=60
    Module.Param0WriteEnabled=true
    Module.Param1WriteEnabled=true
    Module.Param2WriteEnabled=true
    Module.Param3WriteEnabled=true
   )
   &ScaleColor(
    Scale Alpha=/Niagara/DynamicInputs/Scalability/ScalabilityDistanceBasedFloat(
     Curve:NiagaraDataInterfaceCurve=DataInterface() {propertyValues {Curve {keys [{interpMode=RCIM_Cubic, tangentMode=RCTM_User, time=0, value=1}, {interpMode=RCIM_Cubic, tangentMode=RCTM_User, time=0.2, value=0.2}]}}}
     CurveIndex:NiagaraFloat=$Emitter.Scalability.DistanceFraction
     Bias:NiagaraFloat=0
     Scale:NiagaraFloat=1
    )
    Scale RGB=&VectorFromFloat(Value=&Multiply_Float(A=$User.Brightness, B=&FNTimeOfDayScalar(Night Scale=1, Morning Scale=1, Evening Scale=1)))
    ScaleA=true
    ScaleRGB=true
   )
  ]
  rendererClasses [&NiagaraMeshRendererProperties]
  renderers [{rendererClass=&NiagaraMeshRendererProperties, rendererIndex=0}]
 }
 Seam {rendererClasses [], renderers []}
 Sparkles {
  emitterUpdateScript [
   &EmitterScalability(
    Module.bApplySpawnCountScale=false
    Module.bApplySpawnCountScaleByDistanceFraction=false
    Module.bUseMaxDistance=false
    Module.bUseMinDistance=false
   ) {enabled=false}
   &EmitterLifeCycle(NextLoopDuration=4.4208) {enabled=false}
   &EmitterState()
   &SpawnBurst_Instantaneous(Spawn Count=13, Use Spawn Probability=false) {enabled=false}
   &SpawnRate(SpawnRate=6, Use Spawn Probability=false)
  ]
  particleSpawnScript [
   /Niagara/Modules/Spawn/Initialization/InitializeParticle(
    Color=$User.Color
    Lifetime=&UniformRangedFloat(Minimum=0.2, Maximum=1.3)
    Module.Write Color=true
    Module.Write Lifetime=true
    Module.Write Mass=true
    Module.Write Material Random=true
    Module.Write Position=false
    Module.Write Scale=false
    Module.Write SpriteRotation=false
    Module.Write SpriteSize=true
    Position=&SimulationPosition()
    Sprite Size=&Vector2DFromFloat(Value=&UniformRangedFloat(Minimum=10, Maximum=30))
   )
   &SystemLocation(MaskSpawnGroup=false, Offset=$User.GlobalOffset)
   &BoxLocation(Box Offset=<0 50 60>, Box Size=<200 0 0>, MaskSpawnGroup=false, Module.Override Seed=false, Rotate Box=false, Surface Only=false)
   &AddVelocity(Velocity=<0 25 0>)
   /Niagara/Modules/Update/Forces/CurlNoiseForce(Noise Strength=500, Pan Noise=false) {enabled=false}
   &ApplyInitialForces()
  ]
  particleUpdateScript [
   &ParticleState()
   &GravityForce(Gravity=<0 0 -200>)
   &CurlNoiseForce(Noise Frequency=0.02, Noise Strength=150, Pan Noise=false) {module=CurlNoiseForce001}
   &Drag(Drag=3, Use Linear Drag=true, Use Rotational Drag=false)
   &SolveForcesAndVelocity(Clamp Velocity=false, Limit Acceleration=false)
   &ScaleColor(
    Scale Alpha=&FloatFromCurve(FloatCurve=DataInterface() {propertyValues {Curve {keys [<at -0.0092 0.0144>,<at 0.0967 0.9767>,<at 1.0011 0.002>]}}})
    Scale RGB=&VectorFromFloat(Value=2)
    ScaleA=true
    ScaleRGB=true
   )
   &ScaleSpriteSize(Scale Factor=&Vector2DFromFloat(Value=&FloatFromCurve(FloatCurve=DataInterface() {propertyValues {Curve {keys [<at -0.0046 0>,<at 0.1746 0.9952>,<at 1.0011 0.002>]}}})))
  ]
  rendererClasses [&NiagaraSpriteRendererProperties]
  renderers [{rendererClass=&NiagaraSpriteRendererProperties}]
 }
}
summary {
 systemName=NS_Military_Chest_Idle
 emitters [
  {emitterName=Sparkles, rendererClasses [&NiagaraSpriteRendererProperties]}
  {bEnabled=false, emitterName=Rays, rendererClasses [&NiagaraMeshRendererProperties]}
  {emitterName=Seam, rendererClasses []}
 ]
 userVariables [
  User.Brightness {type=NiagaraFloat, defaultValue=1, description=""}
  User.Color {type=LinearColor, description="", defaultValue {struct=LinearColor, value=<rgba 1 0.4148 0 1>}}
  User.GlobalOffset {type=Vector3f, description="", defaultValue {struct=Vector3f, value=<0 0 0>}}
  User.RaysLength {type=NiagaraFloat, defaultValue=0.08, description=""}
 ]
}
